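; Parent quest interactable prop with a helmet that reacts to tagged players and changes their cosmetic variant on completion.
; If the local pawn has the required cosmetic tag, sphere overlaps drive timelines that lift and spin the helmet, then return it, with float and drop audio.
; Quest updates unbind those overlaps.
; Completion shrinks the helmet, while the server completion event fades audio and, after a short delay, switches the interacting player's variant.
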(event EventBeginPlay
  (bind _returnvalue (Fortnite|GetLocalFortPlayerControllers true))
  (bind _output (Utilities|Array|Get(acopy) _returnvalue))
  (bind _returnvalue_2 (Game|Player|GetPlayerPawn _output))
  (bind _sm_helmet (Variables|Default|GetSM_Helmet))
  ;[1 1 1] Init transform vals, sounds and fx
  (|CallParentFunction)
  (Variables|Default|SetInitKeyLoc (Transformation|GetWorldLocation _sm_helmet))
  (Variables|Default|SetInitKeyRot (Transformation|GetWorldRotation _sm_helmet))
  (Variables|Default|SetSuccessFX 0)
  (Variables|Default|SetSuccessSound (Variables|Default|GetNewSuccessSound))
  ;[1 1 1] Bind and Unbind floating when close
  (Utilities|IsValid _output
    (:Is%20Valid
      (Utilities|IsValid _returnvalue_2
        (:Is%20Valid
          (bind _returnvalue_3 (GameplayTags|GetOwnedGameplayTags _returnvalue_2 :!class /Script/GameplayTags.GameplayTagAssetInterface))
          (if (GameplayTags|HasAllTags _returnvalue_3 (Variables|Default|GetCosmeticCustomTag) true)
            (Collision|BindEventtoOnComponentBeginOverlap (Variables|Default|GetOverlapSphereEnter) (EventDispatchers|CreateEvent :!event Custom|EnteredVicinity))
            (Collision|BindEventtoOnComponentEndOverlap (Variables|Default|GetOverlapSphereExit) (EventDispatchers|CreateEvent :!event Custom|ExitedVicinity))))))))

(event EventQuestInitializedAndValid
  (bind _returnvalue (Fortnite|GetLocalFortPlayerControllers true))
  (bind _output (Utilities|Array|Get(acopy) _returnvalue))
  (bind _returnvalue_2 (Game|Player|GetPlayerPawn _output))
  ;[1 1 1] Bind and Unbind floating when close
  (|CallParentFunction)
  (Utilities|IsValid _output
    (:Is%20Valid
      (Utilities|IsValid _returnvalue_2
        (:Is%20Valid
          (bind _returnvalue_3 (GameplayTags|GetOwnedGameplayTags _returnvalue_2 :!class /Script/GameplayTags.GameplayTagAssetInterface))
          (if (GameplayTags|HasAllTags _returnvalue_3 (Variables|Default|GetCosmeticCustomTag) true)
            (Collision|BindEventtoOnComponentBeginOverlap (Variables|Default|GetOverlapSphereEnter) (EventDispatchers|CreateEvent :!event Custom|EnteredVicinity))
            (Collision|BindEventtoOnComponentEndOverlap (Variables|Default|GetOverlapSphereExit) (EventDispatchers|CreateEvent :!event Custom|ExitedVicinity))))))))

(event EventQuestUpdatedFromAnywhere
  ;[1 1 1] Bind and Unbind floating when close
  (|CallParentFunction)
  (Collision|UnbindEventfromOnComponentBeginOverlap (Variables|Default|GetOverlapSphereEnter) (EventDispatchers|CreateEvent :!event Custom|EnteredVicinity))
  (Collision|UnbindEventfromOnComponentEndOverlap (Variables|Default|GetOverlapSphereExit) (EventDispatchers|CreateEvent :!event Custom|ExitedVicinity)))

(event Custom|EnteredVicinity
  (params (OverlappedComponent :type Primitive%20Component%20Object%20Reference) (OtherActor :type Actor%20Object%20Reference) (OtherComp :type Primitive%20Component%20Object%20Reference) (OtherBodyIndex :type Integer) (bFromSweep :type Boolean) (SweepResult :type Hit%20Result%20Structure%20%28by%20ref%29))
  (bind _sm_helmet (Variables|Default|GetSM_Helmet))
  ;[1 1 1] Floating Movement
  (Utilities|FlowControl|Sequence
    (:then_0
      (bind (_direction _lift_time) (Timeline|Lift%20Up PlayFromStart)
        (:Update
          (Transformation|SetWorldLocation _sm_helmet (Math|Vector|Lerp(Vector) (Variables|Default|GetInitKeyLoc) (+ (Variables|Default|GetInitKeyLoc) "0, 0,120.000000") _lift_time) false false)
          (Transformation|SetRelativeRotation _sm_helmet (Math|Rotator|Lerp(Rotator) "-60.000000, 0, 0" "0, 0, 0" _lift_time) false false))
        (:Finished
          (Components|Timeline|SetLooping (Variables|Prop_SuperDuper_Parent|GetSpin) true)
          (bind (_direction_2 _rotate_time _float_time) (Timeline|Spin PlayFromStart)
            (:Update
              (Transformation|SetRelativeLocationAndRotation (Variables|Default|GetSM_Helmet) (Math|Vector|Lerp(Vector) "0, 0,120.000000" "0, 0,125.000000" _float_time) (Math|Rotator|Lerp(Rotator) "0, 0, 0" (Math|Rotator|MakeRotator (.roll "0, 0, 0") (.pitch "0, 0, 0") (+ (.yaw "0, 0, 0") 3.6e+02)) _rotate_time) false false))))))
    (:then_2
      (if (not (Variables|Default|GetIs%20Complete))
        (Audio|Components|Audio|Play (Variables|Default|GetAudio_Key_Float) 0.0)))))

(event Custom|ExitedVicinity
  (params (OverlappedComponent :type Primitive%20Component%20Object%20Reference) (OtherActor :type Actor%20Object%20Reference) (OtherComp :type Primitive%20Component%20Object%20Reference) (OtherBodyIndex :type Integer))
  (bind _sm_helmet (Variables|Default|GetSM_Helmet))
  ;[1 1 1] End Floating Movement
  (Utilities|FlowControl|Sequence
    (:then_0
      ;[1 1 1] Floating Movement
      (Timeline|Spin Stop))
    (:then_1
      (Variables|Default|SetCurrentKeyLoc (Transformation|GetWorldLocation _sm_helmet))
      (Variables|Default|SetCurrentKeyRot (Transformation|GetWorldRotation _sm_helmet))
      (bind (_direction _lift_time) (Timeline|Timeline_0 PlayFromStart)
        (:Update
          (Transformation|SetWorldLocationAndRotation (Variables|Default|GetSM_Helmet) (Math|Vector|Lerp(Vector) (Variables|Default|GetCurrentKeyLoc) (Variables|Default|GetInitKeyLoc) _lift_time) (Math|Rotator|Lerp(Rotator) (Variables|Default|GetCurrentKeyRot) (Variables|Default|GetInitKeyRot) _lift_time) false false))))
    (:then_2
      (Audio|Components|Audio|FadeOut (Variables|Default|GetAudio_Key_Float) 0.25 0.0 "Linear" :!class /Script/Engine.AudioComponent))
    (:then_3
      (Audio|Components|Audio|Play (Variables|Default|GetAudio_Key_Drop) 0.0))))

(event EventThisQuestObjectiveComplete
  (params (IneractingPawn :type Fort%20Pawn%20Object%20Reference))
  ;[1 1 1] Success Movement
  (|CallParentFunction IneractingPawn)
  (bind (_direction _pop_time) (Timeline|FinishQuestPop PlayFromStart)
    (:Update
      (Transformation|SetRelativeScale3D (Variables|Default|GetSM_Helmet) (Math|Vector|Lerp(Vector) "1.000000,1.000000,1.000000" "0.001000,0.001000,0.001000" _pop_time)))))

(event EventServer_QuestObjectiveCompleted
  (params (InteractingPawn :type Fort%20Pawn%20Object%20Reference))
  ;[1 1 1] Update skin
  (|CallParentFunction InteractingPawn)
  (CallFunction|HandleCharacterVariantTransition InteractingPawn)
  (Utilities|FlowControl|Delay 0.5)
  (bind _asathena_player_controller (Utilities|Casting|CastToAthena_PlayerController (Pawn|GetController InteractingPawn)))
  (Class|FortPlayerController|ChangeActiveVariantforCosmeticItem _asathena_player_controller (Variables|Default|GetVariantTemplateID) (Variables|Default|GetVariantChannelToChange) (Variables|Default|GetDesiredActiveVariant) 0))

(event Custom|HandleCharacterVariantTransition
  (params (TransitioningPawn :type Fort%20Pawn%20Object%20Reference))
  ;[1 1 1] Play Update FX
  (Utilities|IsValid TransitioningPawn
    (:Is%20Valid
      (Utilities|IsValid (Variables|/Script/Engine.Character|GetMesh TransitioningPawn)
        (:Is%20Valid
          (Audio|Components|Audio|FadeOut (Variables|Default|GetAudio_Key_Float) 0.25 0.0 "Linear" :!class /Script/Engine.AudioComponent)
          (Variables|Default|SetIs%20Complete true))))))

(fn ConstructionScript
  (|CallParentFunction))
